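; Creative Item Shop device: a vending machine that displays cosmetics from the CosmeticOverride register and opens the store for interacting players.
; The server picks the viewed item, resetting on register or play-mode changes and cycling when damaged.
; ViewingItem's RepNotify async-loads the preview onto a screen material that also reflects enabled state.
; Interaction grants and activates a store-opening gameplay ability; a culture-matched localized texture loads at BeginPlay.
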(event Custom|EnableFunction
  ;[0.2703 0 1 0.2] Receivers / Transmitter
  (Class|CreativeEnabledComponent|SetEnabled (Variables|Default|GetEnabledComponent) true))

(event EventOnDamageServer
  (params (Damage :type Float%20%28single-precision%29) (DamageTags :type Gameplay%20Tag%20Container%20Structure%20%28by%20ref%29) (Momentum :type Vector :default "0, 0, 0") (HitInfo :type Hit%20Result%20Structure%20%28by%20ref%29) (InstigatedBy :type Controller%20Object%20Reference) (DamageCauser :type Actor%20Object%20Reference) (EffectContext :type Gameplay%20Effect%20Context%20Handle%20Structure))
  ;[0 1 0.3626 0.25] Device Interaction
  (Health|SetHealth 1e+03)
  (bind _output_item (CustomizationRegister|GetItemFromRegister false))
  (Variables|Default|SetViewingItem _output_item))

(event EventBeginPlay
  (bind _culture_and_texture_struct (Variables|Default|GetCulture%20and%20Texture%20Struct))
  ;[0 1 0.0488 0.25] Initialization
  (VisualUpdates|UpdateVendingMachineColor "(R=0.725720,G=0.000000,B=1.000000,A=1.000000)")
  ;; Async Load Localized Textures
  (Utilities|FlowControl|Sequence
    (:then_0
      (Networking|CanExecuteCosmeticEvents
        (:True
          (if (Utilities|Array|IsValidIndex _culture_and_texture_struct)
            (bind _localized_texture (CallFunction|ReturnLocTexture _culture_and_texture_struct))
            (if (Utilities|IsValidSoftObjectReference _localized_texture)
              (bind _loaded_object (Creative|Loading|AsyncLoadEditorSafe _localized_texture)
                (:Loaded
                  (bind _astexture (Utilities|Casting|CastToTexture _loaded_object))
                  (Rendering|Material|SetTextureParameterValue (Material|GetMID (Variables|Debug|GetStaticMeshComponent)) (Variables|Default|GetTexture%20Parameter) _astexture))))))))
    (:then_1
      _)))

(event Custom|DisableFunction
  ;[0.2703 0 1 0.2] Receivers / Transmitter
  (Class|CreativeEnabledComponent|SetEnabled (Variables|Default|GetEnabledComponent) false))

(event EventInitializeDevice
  ;[0 1 0.0488 0.25] Initialization
  (if (not (Editor|InEditor))
    (CustomizationRegister|BindToObjectRegister)
    (CustomizationRegister|UpdateFromRegister)
    (CallFunction|SetVisibilityComponent)
    (CallFunction|UpdateSecondaryInteraction) ; We want to make sure that on being placed in creator mode \nthat the store can be accessed
  ))

(event Interaction|EventBlueprintOnInteract
  (params (InteractingPawn :type Fort%20Pawn%20Object%20Reference) (InteractionBeingAttempted :type EInteractionBeingAttempted%20Enum :default "FirstInteraction"))
  ;[0 1 0.3626 0.25] Device Interaction
  (if (CallFunction|IsinPlayMode)
    _
    (elif (Utilities|Enum|Equal(Enum) InteractionBeingAttempted "SecondInteraction")
      _))
  (CallFunction|OpenStoreForPawn InteractingPawn)
  (bind _asfort_player_controller_zone (Utilities|Casting|CastToFortPlayerControllerZone (Fortnite|GetFortPlayerControllerfromActor InteractingPawn)))
  (Cosmetics|AllowLoadoutRefresh _asfort_player_controller_zone))

(event Custom|Object%20Added%20to%20Register
  (params (Object%20Added :type Object%20Reference) (Index :type Integer))
  ;[0.1455 1 0.8701 0.25] Register Updates
  (CustomizationRegister|UpdateFromRegister))

(event ChildFunctions|EventPlayModeChanged
  (params (Minigame :type Fort%20Minigame%20Object%20Reference) (bIsInPlayMode :type Boolean))
  ;[0 0.8698 0.7627 0.25] State Changes
  (|CallParentFunction Minigame bIsInPlayMode)
  (CallFunction|UpdateSecondaryInteraction)
  (Networking|SwitchHasAuthority
    (:Authority
      (bind _output_item (CustomizationRegister|GetItemFromRegister true))
      (Variables|Default|SetViewingItem _output_item))))

(event Custom|Object%20Removed%20from%20Register
  (params (Object%20Removed :type Object%20Reference) (Index :type Integer))
  ;[0.1455 1 0.8701 0.25] Register Updates
  (CustomizationRegister|UpdateFromRegister))

(event OnEnabledStateChanged(EnabledComponent)
  (params (Enabled :type Boolean))
  ;[0 0.8698 0.7627 0.25] State Changes
  (VisualUpdates|UpdateScreen)
  (Visibility|SetVisiblityDuringPhase (Variables|Default|GetVisibleInGameComponent) (select Enabled "NewEnumerator3" "NewEnumerator1")))

(event Custom|LoadItem
  (params (Target :type Fort%20Item%20Definition%20Object%20Reference))
  (bind _returnvalue (Item|GetLargePreviewImage Target))
  ;[1 0.8402 0 0.25] Asyncronous Loading of a New Image for the Screen, then Updating Screen
  (Utilities|IsValid Target
    (:Is%20Valid
      (if (Utilities|IsValidSoftObjectReference _returnvalue)
        (bind _object (Utilities|AsyncLoadAsset _returnvalue)
          (:Completed
            (bind _astexture_2d (Utilities|Casting|CastToTexture2D _object))
            (Variables|Default|SetCachedScreenImage _astexture_2d)))
        (else
          (Variables|Default|SetCachedScreenImage "") ; Invalid - Update with empty.
        ))
      (VisualUpdates|UpdateScreen))
    (:Is%20Not%20Valid
      _)))

(fn BindtoObjectRegister
  (bind _cosmeticoverride_register (Variables|Default|GetCosmeticOverride%20Register))
  (CreativeObjectRegistry|GetOrCreateObjectRegister self "(TagName=\"Creative.Register.CosmeticOverride\")" false 0 (Variables|Default|GetCosmeticOverride%20Register))
  (Default|BindEventtoOnObjectAdded _cosmeticoverride_register (EventDispatchers|CreateEvent :!event Object%20Added%20to%20Register))
  (Default|BindEventtoOnObjectRemoved _cosmeticoverride_register (EventDispatchers|CreateEvent :!event Object%20Removed%20from%20Register)))

(fn BlueprintCanInteract
  (params (InteractingPawn :type Fort%20Pawn%20Object%20Reference) (InteractionBeingAttempted :type EInteractionBeingAttempted%20Enum :default "FirstInteraction") (InteractionType :type TInteractionType%20Enum :default "IT_NoInteraction"))
  (if (CallFunction|IsinPlayMode)
    _
    (elif (Utilities|Enum|Equal(Enum) InteractionBeingAttempted "SecondInteraction")
      _
      (else
        (return true))))
  (return (Class|CreativeEnabledComponent|IsEnabled (Variables|Default|GetEnabledComponent))))

(fn BlueprintGetInteractionString
  (params (InteractingPawn :type Fort%20Pawn%20Object%20Reference) (InteractionBeingAttempted :type EInteractionBeingAttempted%20Enum :default "FirstInteraction"))
  ;[0 0.8698 0.7627 0.25] If we're in Create mode, just return the normal string if we're the first interaction. If we're in play mode or the second interact, send an appropriate string based on what will happen when you interact (either opens the store in general or at a specific page).
  (if (CallFunction|IsinPlayMode)
    _
    (elif (Utilities|Enum|Equal(Enum) (Variables|GetInteractionBeingAttempted) "SecondInteraction")
      _
      (else
        (bind _returnvalue (|CallParentFunction (Variables|GetInteractingPawn) (Variables|GetInteractionBeingAttempted)))
        (return _returnvalue))))
  (bind _viewingitem (Variables|Default|GetViewingItem)
    (:then
      (return (Utilities|Text|FormatText "NSLOCTEXT(\"[17F717074084665DB1789986FDFEFA2D]\", \"00D5A9E24BFF570550C50BA4FF70D7E7\", \"Open Store: {Item}\")" (ItemDefinition|GetDisplayName _viewingitem))))
    (:else
      (return "NSLOCTEXT(\"[17F717074084665DB1789986FDFEFA2D]\", \"563DE2EB4787A3667F7057ABCC0B4A5E\", \"Open Store\")"))))

(fn BlueprintGetInteractionTime
  (params (InteractingPawn :type Fort%20Pawn%20Object%20Reference) (InteractionBeingAttempted :type EInteractionBeingAttempted%20Enum :default "FirstInteraction"))
  ;[0.2159 0.956 1] Override to set our interaction time to user option settings
  (bind _fortminigameprogress (Variables|Default|GetFortMinigameProgress)
    (:then
      (bind _currentminigame (Variables|/Script/FortniteGame.FortMinigameLogicComponent|GetCurrentMinigame _fortminigameprogress)
        (:then
          (return (Math|Float|Max(Float) (Variables|UserOptions|GetInteractTime)) (Minigame|IsinPlayMode _currentminigame :!class /Script/FortniteGame.FortMinigame)))
        (:else
          _)))
    (:else
      _))
  (return 0.0 false))

(fn GetItemfromRegister
  (params (FirstItem :type Boolean))
  ;[0 0.8698 0.7627 0.25] Gathers the contents of the appropriate register (which will contain mannequins), then gets an indexed option from it. It then uses the Customization Interface to get details from this device.
  ; If this is the Server it will also update the Current Register
  (bind (_bisvalid _register) (Class|FortCreativeIndexedObjectRegisterComponent|GetRegisteratIndex (Variables|Default|GetCosmeticOverride%20Register) (Variables|Default|GetCustomizationSlot)))
  (bind _register_objects (Utilities|Struct|BreakFortObjectArray _register))
  (if _bisvalid
    (Networking|SwitchHasAuthority
      (:Authority
        (Variables|Default|SetCurrentRegisterIndex (Math|Integer|SelectInt 0 (Math|Integer|%25%28Integer%29 (+ (Variables|Default|GetCurrentRegisterIndex) 1) (Utilities|Array|Length _register_objects)) (Variables|GetFirstItem))))
      (:Remote
        _))
    (bind _cosmetics__male_or_all (CreativeOverrideCustomization|GetCreativeCustomizationData(Message) (Utilities|Array|Get(acopy) _register_objects (Variables|Default|GetCurrentRegisterIndex)) self))
    (if (Utilities|Array|IsValidIndex _cosmetics__male_or_all)
      (bind _output_item (Utilities|Struct|BreakFortSwapItemAndVariantData (Utilities|Array|Get(acopy) _cosmetics__male_or_all)))
      (return _output_item))
    (else
      (Variables|Default|SetCurrentRegisterIndex 0)
      (return 0))))

(fn MigrateLegacyUserOptions
  ;[0.2159 0.956 1] Migrate Legacy User Options
  (|CallParentFunction)
  (Utilities|FlowControl|Sequence
    (:then_0
      (bind _migratedproperties (DeviceConversion|MigrateLegacyEnabledIndex (Variables|UserOptions|Basic|GetEnabledOnPhase) "NewEnumerator4"))
      (LocalVariables|SetOptionMigrated _migratedproperties))
    (:then_1
      (if (LocalVariables|GetOptionMigrated)
        (OptionData|RefreshPropertyOverrides (Variables|Default|GetToyOptionsComponent))))))

(fn OnViewingItemChanged
  ;[1 1 1] On Viewing Item Changed
  (Networking|CanExecuteCosmeticEvents
    (:True
      (CallFunction|UpdateSecondaryInteraction)
      (CallFunction|LoadItem (Variables|Default|GetViewingItem)))))

(fn OnRep_NumRegisterItems)

(fn OnRep_ViewingItem
  ;[1 1 1] OnRep_ViewingItem
  (VisualUpdates|OnViewingItemChanged))

(fn OpenStoreforPawn
  (params (Pawn :type Pawn%20Object%20Reference))
  ;[0.1455 1 0.8701 0.25] Server Only - If Enabled will apply a GA to the Pawn that opens the store for their client.
  (Utilities|IsValid Pawn
    (:Is%20Valid
      (Networking|SwitchHasAuthority
        (:Authority
          (if (Class|CreativeEnabledComponent|IsEnabled (Variables|Default|GetEnabledComponent))
            (Ability|Advanced|GrantandActivateAbilityonSelfwithParams (Ability|GetAbilitySystemComponent Pawn) "/CreativeOpenFrontEndShop/GAT_Creative_OpenFrontEnd_Shop.GAT_Creative_OpenFrontEnd_Shop_C" self (Utilities|Struct|MakeGameplayEventData "(TagName=\"Creative.Event.OpenFrontEnd.Store\")" self))))))))

(fn ReturnLocTexture
  (params (Localization%20Struct :type Array%20of%20Culture%20Texture%20Struct%20Structures))
  ;[0.2083 0.9901 1] Cycle through the soft references of our textures and find the one that corresponds to the localization settings
  (for _array_element Localization%20Struct
    (if (Utilities|String|EqualExactly(String) (Utilities|Internationalization|GetCurrentCulture) (.Culture_11_57E9F25741437CC55361B58F9DB743E0 (Utilities|Struct|BreakCultureTextureStruct _array_element)))
      (return (Utilities|Struct|BreakCultureTextureStruct _array_element))))
  (return ""))

(fn SetVisibilityComponent
  (bind _scene_components (Variables|Default|GetScene%20Components))
  ;[0.9264 1 0.2586] Set the static mesh to visibility selection so we can tie it to enabled states
  (bind _staticmeshcomponent (Variables|Debug|GetStaticMeshComponent)
    (:then
      (Utilities|Array|Add _scene_components _staticmeshcomponent)
      (bind _visibleingamecomponent (Variables|Default|GetVisibleInGameComponent)
        (:then
          (Visibility|SetVisibilityComponents _visibleingamecomponent _scene_components true))
        (:else)))
    (:else)))

(fn UpdatefromRegister
  ;[0 0.8698 0.7627 0.25] When the register updates, if we're the Server reset the Register. If we're the client update the number of items.
  (Networking|SwitchHasAuthority
    (:Authority
      (bind _output_item (CustomizationRegister|GetItemFromRegister true))
      (Variables|Default|SetViewingItem _output_item))
    (:Remote
      (VisualUpdates|UpdateNumberOfItems))))

(fn UpdateNumberofItems
  ;[0.2159 0.956 1] Update Number of Items
  (bind (_bisvalid _register) (Class|FortCreativeIndexedObjectRegisterComponent|GetRegisteratIndex (Variables|Default|GetCosmeticOverride%20Register) 0))
  (bind _register_objects (Utilities|Struct|BreakFortObjectArray _register))
  (Rendering|Material|SetScalarParameterValue (Variables|Default|GetScreen%20MID) "NumItems" (Math|Conversions|ToFloat(Integer) (Utilities|Array|Length _register_objects)) :!class /Script/Engine.MaterialInstanceDynamic))

(fn UpdateScreen
  (bind _enabledcomponent (Variables|Default|GetEnabledComponent))
  (bind _enabled (Class|CreativeEnabledComponent|IsEnabled _enabledcomponent))
  (bind _screen_mid (Variables|Default|GetScreen%20MID))
  ;[0.2159 0.956 1] Update Screen
  (Networking|CanExecuteCosmeticEvents
    (:True
      (Material|ReattachOrCreateMID (Variables|Debug|GetStaticMeshComponent) (Variables|Default|GetScreen%20MID))
      (Rendering|Material|SetTextureParameterValue _screen_mid "WeaponIcon" (select _enabled (select (Utilities|IsValid (Variables|Default|GetCachedScreenImage) :!class /Script/Engine.KismetSystemLibrary) (Variables|Default|GetCachedScreenImage) "/Game/UI/Icons/WorldIcon_Heart.WorldIcon_Heart") "/Game/UI/Foundation/Textures/Icons/ItemTypes/T-Icon-Blank.T-Icon-Blank"))
      (Rendering|Material|SetVectorParameterValue _screen_mid "ScreenColor" (Math|Color|SelectColor "(R=0.911458,G=0.735922,B=0.000000,A=1.000000)" "(R=0.039931,G=0.001753,B=0.000000,A=1.000000)" _enabled) :!class /Script/Engine.MaterialInstanceDynamic)
      (Rendering|Material|SetVectorParameterValue (Variables|Default|GetScreen%20MID) "ItemSelectionVector" (Utilities|Select :Index (Math|Integer|%25%28Integer%29 (Variables|Default|GetCurrentRegisterIndex) 2) :Option%200 "(R=1.000000,G=0.000000,B=0.000000,A=1.000000)" :Option%201 "(R=0.000000,G=1.000000,B=0.000000,A=1.000000)" :Option%202 "(R=0.000000,G=0.000000,B=1.000000,A=1.000000)") :!class /Script/Engine.MaterialInstanceDynamic))))

(fn UpdateSecondaryInteraction)

(fn UpdateVendingMachineColor
  (params (New%20Color :type Linear%20Color%20Structure :default "(R=0.725720,G=0.000000,B=1.000000,A=1.000000)"))
  (bind _base_mid (Variables|Default|GetBase%20MID))
  ;[0.2159 0.956 1] Update Vending Machine Color
  (Material|ReattachOrCreateMID (Variables|Debug|GetStaticMeshComponent) _base_mid)
  (Rendering|Material|SetVectorParameterValue _base_mid "PaintColor" New%20Color :!class /Script/Engine.MaterialInstanceDynamic))

(fn ConstructionScript
  ;[1 1 1] Constructor
  (|CallParentFunction))
